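# Animation Blueprint for the ScribeSpider backpack that plays its idle animation and offsets the pack to suit the wearer.
# An Idle state machine feeds an additive Modify Bone on "pack", driven by BackDistance and BackHeight.
# Each update, pawn name matches for ScribeSpider, SaltyBrim or MixedAlley get fixed distance and height offsets.
# Other characters get an interpolated distance chosen by gender, body type and property-access flags.
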AnimBlueprint /BRCosmetics/Accessories/FORT_Backpacks/Backpack_ScribeSpider/Meshes/M_MED_ScribeSpider_Pack_AnimBP
target_skeleton=&/M_MED_ScribeSpider_Pack_Skeleton
graphs [
 AnimGraph {nodes [
   AnimGraphNode_StateMachine_0 = Idle() {is=title AnimGraphNode_StateMachine}
   n1 = "Transform (Modify) Bone - Bone: pack"!(
    ComponentPose=Local To Component(LocalPose=@AnimGraphNode_StateMachine_0)
    Alpha
    Translation_X:double=0
    Translation_Y:double=Get BackDistance() -> (BackDistance:double)
    Translation_Z:double=Get BackHeight() -> (BackHeight:double)
    Rotation="0,0,0"
    Scale
   ) -> (Pose) {
    rotationMode=BMM_Additive
    rotationSpace=BCS_BoneSpace
    scaleMode=BMM_Additive
    translationMode=BMM_Additive
    alphaBoolBlend {blendOption=Linear}
    boneToModify {boneName=pack}
   }
   Output Pose(Result=Component To Local(ComponentPose=@n1)) {layerGroup=None}
  ]}
 Idle {entry=Idle, path=AnimGraph.AnimGraphNode_StateMachine_0.Idle, ids=id, nodes [AnimStateNode_0 = Idle() {is=title AnimStateNode}]}
 Idle {
  path=AnimGraph.AnimGraphNode_StateMachine_0.Idle.AnimStateNode_0.Idle
  nodes [Output Animation Pose(Result=Sequence Player 'ScribeSpider_Idle_pack'() {sequence=/BRCosmetics/Accessories/FORT_Backpacks/Backpack_ScribeSpider/Anim/ScribeSpider_Idle_pack})]
 }
 EventGraph {nodes [
   Event BlueprintUpdateAnimation()
   > GetIPSkinCharacter()
   > n1 = SetBackDistance!(execute, self) -> (then, FinalDistance:double, FinalHeight:double)
   > Set BackDistance(BackDistance:double=@n1) -> (Output_Get:double)
  ]}
 GetIPSkinCharacter {nodes [
   GetIPSkinCharacter() {is=title K2Node_FunctionEntry}
   n1 = GetCharacterDisplayName!(Actor:Actor Object Reference=TryGetPawnOwner()) -> (ReturnValue:String)
   > n2 = Branch(Condition=OR Boolean!(
     A=Contains(SearchIn=@n1, Substring=ScribeSpider)
     B=Contains(SearchIn=@n1, Substring=SaltyBrim)
     C:bool=Contains(SearchIn=@n1, Substring=MixedAlley)
    ) -> (ReturnValue))
   > Set IPSkinCharacter(IPSkinCharacter:bool=true) -> (Output_Get:bool)
   Set IPSkinCharacter(execute=@n2.else, IPSkinCharacter=false)
  ]}
 SetBackDistance {nodes [
   SetBackDistance() {is=title K2Node_FunctionEntry}
   > n1 = Branch(Condition=Get IPSkinCharacter() -> (IPSkinCharacter:bool))
   > n2 = Set DistanceValue(DistanceValue:double=-2.5) -> (Output_Get:double)
   > n3 = Set BackHeight(BackHeight:double=-6) -> (Output_Get:double)
   > Return Node!(execute, FinalDistance:double=@n2, FinalHeight:double=@n3)
   n4 = Branch(
    execute=@n1.else
    Condition="Equal (Enum)"(
     A:EFortCustomGender Enum=Property Access!() -> (Value:EFortCustomGender Enum) {is=title K2Node_PropertyAccess}
     B:EFortCustomGender Enum=Female
    )
   )
   n5 = Property Access()
   > n6 = Branch(Condition=AND Boolean(A=Property Access() -> (Value:bool=~), B="Equal (Enum)"(A=@n5, B=Female)))
   > n7 = Set DistanceValue(DistanceValue=0)
   Return Node(
    execute=[@n7.then, @n8.then]
    FinalDistance=FInterpTo(
     Current=Get BackDistance()
     Target=Get DistanceValue() -> (DistanceValue:double)
     DeltaTime=GetWorldDeltaSeconds!() -> (ReturnValue:double)
     InterpSpeed=5
    )
    FinalHeight=0
   )
   n8 = Set DistanceValue(execute=@n6.else, DistanceValue=0.4)
   n9 = "Equal (Enum)"(A=Property Access() -> (Value:EFortCustomGender Enum), B=Male)
   n10 = "Equal (Enum)"(A:EFortPlayerAnimBodyType Enum=Property Access() -> (Value:EFortPlayerAnimBodyType Enum), B:EFortPlayerAnimBodyType Enum=Large)
   n11 = Branch(execute=@n4.else, Condition=AND Boolean(A=Property Access() -> (Value:bool=~), B=OR Boolean!(A=@n10, B=@n9) -> (ReturnValue)))
   > n12 = Set DistanceValue(DistanceValue=-1.5)
   Return Node(
    execute=[@n13.then, @n12.then]
    FinalDistance=FInterpTo(Current=Get BackDistance(), Target=Get DistanceValue(), DeltaTime=GetWorldDeltaSeconds(), InterpSpeed=5)
    FinalHeight=0
   )
   n13 = Set DistanceValue(execute=@n11.else, DistanceValue=0)
  ]}
]
variables [
 BackDistance {type=number}
 BackHeight {type=number}
 IPSkinCharacter {type=boolean}
]
